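; Drive-in screen Blueprint that shows an hours, minutes and seconds countdown to calendar event "R0" through dynamic material digit parameters.
; On authority, InitializeLevel watches the event, and ContinueCountdownEvent converts time remaining into a server-world end time.
; Its OnRep starts a one-second timer on remotes that splits reversed strings into digits, folding days into hours.
; On authority, OnRep schedules a late accuracy timer, whose handler is empty.
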(event Collision|EventActorBeginOverlap
  (params (OtherActor :type Actor%20Object%20Reference)))

(event EventBeginPlay)

(event Custom|ContinueCountdownEvent
  ;[1 1 1] Countdown Initialize
  (Variables|Default|SetServerWorldSecondsAtLaunch (+ (GameState|GetServerWorldTimeSeconds (Game|GetGameState)) (Math|Timespan|GetTotalSeconds (Variables|Default|GetTimeUntilCountdownEnd)))))

(event Custom|InitializeLevel
  ;[1 1 1] Countdown Initialize
  (Networking|SwitchHasAuthority
    (:Authority
      (bind (_eventname _timeuntilend) (MCPContext|WatchCalendarEvent "R0")
        (:OnEventActive
          (Variables|Default|SetTimeUntilCountdownEnd _timeuntilend)
          (CallFunction|ContinueCountdownEvent))
        (:OnEventUpdated
          (Variables|Default|SetTimeUntilCountdownEnd _timeuntilend))))))

(event Custom|CountdownTimer_New
  (bind _timespanremaining (Variables|Default|GetTimespanRemaining))
  (bind _returnvalue (Math|Timespan|GetHours _timespanremaining))
  (bind _returnvalue_2 (Math|Timespan|GetDays _timespanremaining))
  (bind _returnvalue_3 (* _returnvalue_2 24))
  (bind _returnvalue_4 (+ _returnvalue _returnvalue_3))
  (bind _returnvalue_5 (Utilities|String|ToString(Integer) _returnvalue_4))
  (bind _returnvalue_6 (Math|Timespan|GetSeconds _timespanremaining))
  (bind _returnvalue_7 (Utilities|String|ToString(Integer) _returnvalue_6))
  (bind _returnvalue_8 (Math|Timespan|GetMinutes _timespanremaining))
  (bind _returnvalue_9 (Utilities|String|ToString(Integer) _returnvalue_8))
  (bind _returnvalue_10 (Utilities|String|Reverse _returnvalue_9))
  (bind _returnvalue_11 (Utilities|String|Reverse _returnvalue_7))
  (bind _returnvalue_12 (Utilities|String|Reverse _returnvalue_5))
  (bind _countdownmaterial (Variables|Default|GetCountdownMaterial))
  ;[1 1 1] Update Countdown every second
  (Variables|Default|SetSecondsRemaining (Utilities|ClampMinInteger (+ (Math|Float|Truncate (- (Variables|Default|GetServerWorldSecondsAtLaunch) (GameState|GetServerWorldTimeSeconds (Game|GetGameState)))) (Variables|Default|GetTimeOffsetSeconds))))
  ;[1 1 1] Update Visuals
  (Variables|Default|SetTimespanRemaining (Math|Timespan|MakeTimespan 0 0 0 (Variables|Default|GetSecondsRemaining)))
  (Rendering|Material|SetScalarParameterValue _countdownmaterial "ShowCountdown" 1.0 :!class /Script/Engine.MaterialInstanceDynamic) ; Show Countdown or blank movie screen
  (Rendering|Material|SetScalarParameterValue _countdownmaterial "Digit_1" (Utilities|String|StringToFloat (Utilities|String|GetSubstring _returnvalue_12 1)) :!class /Script/Engine.MaterialInstanceDynamic) ; Digit
  (Rendering|Material|SetScalarParameterValue _countdownmaterial "Digit_2" (Utilities|String|StringToFloat (Utilities|String|GetSubstring _returnvalue_12)) :!class /Script/Engine.MaterialInstanceDynamic) ; Digit
  (Rendering|Material|SetScalarParameterValue _countdownmaterial "Digit_4" (Utilities|String|StringToFloat (Utilities|String|GetSubstring _returnvalue_10 1)) :!class /Script/Engine.MaterialInstanceDynamic) ; Digit
  (Rendering|Material|SetScalarParameterValue _countdownmaterial "Digit_5" (Utilities|String|StringToFloat (Utilities|String|GetSubstring _returnvalue_10)) :!class /Script/Engine.MaterialInstanceDynamic) ; Digit
  (Rendering|Material|SetScalarParameterValue _countdownmaterial "Digit_7" (Utilities|String|StringToFloat (Utilities|String|GetSubstring _returnvalue_11 1)) :!class /Script/Engine.MaterialInstanceDynamic) ; Digit
  (Rendering|Material|SetScalarParameterValue _countdownmaterial "Digit_8" (Utilities|String|StringToFloat (Utilities|String|GetSubstring _returnvalue_11)) :!class /Script/Engine.MaterialInstanceDynamic) ; Digit
)

(event Custom|StartFinalAccuracyadjustment
  ;[1 1 1] Start the accuracy adjustment
  (bind _returnvalue (Utilities|Time|SetTimerbyEvent (EventDispatchers|CreateEvent :!event Custom|CountDownTimer_AccuracyFix) 15.0 true))
  (Variables|Default|SetCountDownTimerAccuracyfix _returnvalue))

(event Custom|CountDownTimer_AccuracyFix)

(event Custom|ErrorCorrection
  ;[1 1 1] Start second timer for more accuracy
  (Utilities|FlowControl|DoOnce false
    (:Completed
      (Utilities|Time|SetTimerbyEvent (EventDispatchers|CreateEvent :!event Custom|StartFinalAccuracyadjustment) (Math|Float|Clamp(Float) (Math|Conversions|ToFloat(Integer) (- (Utilities|ClampMinInteger (+ (Math|Float|Truncate (- (Variables|Default|GetServerWorldSecondsAtLaunch) (GameState|GetServerWorldTimeSeconds (Game|GetGameState)))) (Variables|Default|GetTimeOffsetSeconds))) 300)) 0.001 1e+10)))))

(fn OnRep_ServerWorldSecondsAtLaunch
  (Networking|SwitchHasAuthority
    (:Authority
      (CallFunction|ErrorCorrection))
    (:Remote
      (if (not (Utilities|Time|IsTimerActivebyHandle (Variables|Default|GetCountdownTimerHandle)))
        (Variables|Default|SetSecondsRemaining (Utilities|ClampMinInteger (+ (Math|Float|Truncate (- (Variables|Default|GetServerWorldSecondsAtLaunch) (GameState|GetServerWorldTimeSeconds (Game|GetGameState)))) (Variables|Default|GetTimeOffsetSeconds))))
        (bind _returnvalue (Utilities|Time|SetTimerbyEvent (EventDispatchers|CreateEvent :!event CountdownTimer_New) 1.0 true))
        (Variables|Default|SetCountdownTimerHandle _returnvalue)))))

(fn ConstructionScript
  (bind _countdownmaterial (Variables|Default|GetCountdownMaterial))
  (bind _returnvalue (Rendering|Material|CreateDynamicMaterialInstance (Variables|Default|GetDriveIn_Screen_Fixed) 0 0 "None" :!class /Script/Engine.PrimitiveComponent))
  (Variables|Default|SetCountdownMaterial _returnvalue)
  (Rendering|Material|SetScalarParameterValue _countdownmaterial "Digit_1" 0.0 :!class /Script/Engine.MaterialInstanceDynamic) ; Digit
  (Rendering|Material|SetScalarParameterValue _countdownmaterial "Digit_2" 0.0 :!class /Script/Engine.MaterialInstanceDynamic) ; Digit
  (Rendering|Material|SetScalarParameterValue _countdownmaterial "Digit_4" 0.0 :!class /Script/Engine.MaterialInstanceDynamic) ; Digit
  (Rendering|Material|SetScalarParameterValue _countdownmaterial "Digit_5" 0.0 :!class /Script/Engine.MaterialInstanceDynamic) ; Digit
  (Rendering|Material|SetScalarParameterValue _countdownmaterial "Digit_7" 0.0 :!class /Script/Engine.MaterialInstanceDynamic) ; Digit
  (Rendering|Material|SetScalarParameterValue _countdownmaterial "Digit_8" 0.0 :!class /Script/Engine.MaterialInstanceDynamic) ; Digit
)
